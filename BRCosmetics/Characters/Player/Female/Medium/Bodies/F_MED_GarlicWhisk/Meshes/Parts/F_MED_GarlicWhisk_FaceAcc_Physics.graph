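# Physics asset for the GarlicWhisk female medium character's face accessory mesh, providing physics bodies for dangling left and right earrings.
# Each capsule earring joins a small sphere base with limited swing and twist, and each base is angularly locked to the head.
# Head and two neck capsules, chained by angularly free constraints, give the body volume around the earrings.
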PhysicsAsset /BRCosmetics/Characters/Player/Female/Medium/Bodies/F_MED_GarlicWhisk/Meshes/Parts/F_MED_GarlicWhisk_FaceAcc_Physics
bodies {
 dyn_l_earring [{
   shapeType=Capsule
   length=2.3714
   radius=1.0484
   shapeName=dyn_l_earring_capsule
   center=<3 -0.0257 0.0055>
   rotation=<rot 0 -90.3127 -89.3856>
  }]
 dyn_l_earring_BASE [{shapeType=Sphere, radius=0.505, shapeName=dyn_l_earring_BASE_sphere}]
 dyn_r_earring [{
   shapeType=Capsule
   length=2.3714
   radius=1.0484
   shapeName=dyn_r_earring_capsule
   center=<-1 0.0257 -0.0055>
   rotation=<rot 0 -90.3127 -89.3856>
  }]
 dyn_r_earring_BASE [{shapeType=Sphere, radius=0.505, shapeName=dyn_r_earring_BASE_sphere}]
 head [{
   shapeType=Capsule
   length=7.117
   radius=5.6368
   shapeName=head_capsule
   center=<3.453 1.7354 -0.0194>
   rotation=<rot -0.304 50.2916 -90.1106>
  }]
 neck_01 [{shapeType=Capsule, length=6.505, radius=5.005, shapeName=neck_01_capsule, rotation=<rot 0 -90.0001 90>}]
 neck_02 [{shapeType=Capsule, length=6.505, radius=5.005, shapeName=neck_02_capsule, rotation=<rot 0 -90.0001 90>}]
}
constraints [
 {
  bone1Name=head
  bone2Name=neck_02
  swing1LimitDegrees=45
  swing1Motion=Free
  swing2LimitDegrees=45
  swing2Motion=Free
  twistLimitDegrees=45
  twistMotion=Free
 }
 {
  bone1Name=neck_02
  bone2Name=neck_01
  swing1LimitDegrees=45
  swing1Motion=Free
  swing2LimitDegrees=45
  swing2Motion=Free
  twistLimitDegrees=45
  twistMotion=Free
 }
 {bone1Name=dyn_l_earring, bone2Name=dyn_l_earring_BASE, swing1LimitDegrees=45, swing2LimitDegrees=45, twistLimitDegrees=45}
 {bone1Name=dyn_r_earring, bone2Name=dyn_r_earring_BASE, swing1LimitDegrees=45, swing2LimitDegrees=45, twistLimitDegrees=45}
 {
  bone1Name=dyn_l_earring_BASE
  bone2Name=head
  swing1LimitDegrees=45
  swing1Motion=Locked
  swing2LimitDegrees=45
  swing2Motion=Locked
  twistLimitDegrees=45
  twistMotion=Locked
 }
 {
  bone1Name=dyn_r_earring_BASE
  bone2Name=head
  swing1LimitDegrees=45
  swing1Motion=Locked
  swing2LimitDegrees=45
  swing2Motion=Locked
  twistLimitDegrees=45
  twistMotion=Locked
 }
]
